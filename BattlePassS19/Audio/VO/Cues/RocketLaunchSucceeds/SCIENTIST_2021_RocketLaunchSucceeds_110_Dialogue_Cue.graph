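# A Sound Cue that plays the Scientist's voice line for a successful rocket launch in the Battle Pass Season 19 content.
# A single Dialogue Player node plays the dialogue wave with the Scientist as speaker and no listener targets.
# The cue routes to the narrator VO sound class, uses the season's quest attenuation with a very large max distance, and boosts volume.
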SoundCue /BattlePassS19/Audio/VO/Cues/RocketLaunchSucceeds/SCIENTIST_2021_RocketLaunchSucceeds_110_Dialogue_Cue
root=DialoguePlayer_0
nodes [{
  id=DialoguePlayer_0
  properties {dialogueWaveParameter {
    dialogueWave=/BattlePassS19/Audio/VO/Waves/SCIENTIST_2021_RocketLaunchSucceeds_110_Dialogue
    context {speaker=/BattlePassS19/Audio/VO/DV_Scientist, targets []}
   }}
 }]
properties {
 attenuationSettings=/BattlePassS19/Audio/VO/S19_Quest_Att
 maxDistance=2097152
 soundClassObject=/Game/Sounds/FortSoundClassesAndModes/Fort_VO_Narrator
 volumeMultiplier=2
}
